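# MurkyVole jetpack movement foley cue, layering a random movement sound with a random servo sound.
# An IsLocal branch selects first-person or third-person footstep attenuation over a shared, slightly pitch-randomized mix.
# Each layer switches by quality level between a six-variant pool and a three-variant subset, with the servo layer passing through its own modulator.
# The cue plays at reduced volume under the player foley sound class with its own concurrency cap.
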SoundCue /MurkyVoleGameplay/Assets/Audio/JetPack/MurkyVole_Foley_Cue
root=Branch_0
nodes [
 {id=Attenuation_0, children [Modulator_0], properties {attenuationSettings=/Game/Sounds/Fort_Player/Footsteps/Settings/Fort_Player_FS_1P_Att}}
 {id=Attenuation_1, children [Modulator_0], properties {attenuationSettings=/Game/Sounds/Fort_Player/Footsteps/Settings/Fort_Player_FS_3P_Att}}
 {id=Branch_0, children [Attenuation_0, Attenuation_1, Attenuation_1], properties {boolParameterName=IsLocal}}
 {id=Mixer_0, children [QualityLevel_1, Modulator_1], properties {inputVolume [1,1]}}
 {id=Modulator_0, children [Mixer_0], properties {pitchMax=1.1, pitchMin=1}}
 {id=Modulator_1, children [QualityLevel_2]}
 {id=QualityLevel_1, children [Random_2,Random_2,Random_3,Random_3]}
 {id=QualityLevel_2, children [Random_5,Random_5,Random_6,Random_6]}
 {
  id=Random_2
  children [WavePlayer_9, WavePlayer_10, WavePlayer_11, WavePlayer_12, WavePlayer_13, WavePlayer_14]
  properties {weights [1,1,1,1,1,1]}
 }
 {id=Random_3, children [WavePlayer_12, WavePlayer_13, WavePlayer_14], properties {bShouldExcludeFromBranchCulling=true, weights [1,1,1]}}
 {
  id=Random_5
  children [WavePlayer_19, WavePlayer_20, WavePlayer_21, WavePlayer_22, WavePlayer_23, WavePlayer_24]
  properties {weights [1,1,1,1,1,1]}
 }
 {id=Random_6, children [WavePlayer_22, WavePlayer_23, WavePlayer_24], properties {bShouldExcludeFromBranchCulling=true, weights [1,1,1]}}
 {id=WavePlayer_10, properties {soundWaveAssetPtr=&/MurkyVole_Movement_Foley_02}}
 {id=WavePlayer_11, properties {soundWaveAssetPtr=&/MurkyVole_Movement_Foley_03}}
 {id=WavePlayer_12, properties {soundWaveAssetPtr=&/MurkyVole_Movement_Foley_04}}
 {id=WavePlayer_13, properties {soundWaveAssetPtr=&/MurkyVole_Movement_Foley_05}}
 {id=WavePlayer_14, properties {soundWaveAssetPtr=&/MurkyVole_Movement_Foley_06}}
 {id=WavePlayer_19, properties {soundWaveAssetPtr=&/MurkyVole_Movement_Servo_Foley_01}}
 {id=WavePlayer_20, properties {soundWaveAssetPtr=&/MurkyVole_Movement_Servo_Foley_02}}
 {id=WavePlayer_21, properties {soundWaveAssetPtr=&/MurkyVole_Movement_Servo_Foley_03}}
 {id=WavePlayer_22, properties {soundWaveAssetPtr=&/MurkyVole_Movement_Servo_Foley_04}}
 {id=WavePlayer_23, properties {soundWaveAssetPtr=&/MurkyVole_Movement_Servo_Foley_05}}
 {id=WavePlayer_24, properties {soundWaveAssetPtr=&/MurkyVole_Movement_Servo_Foley_06}}
 {id=WavePlayer_9, properties {soundWaveAssetPtr=&/MurkyVole_Movement_Foley_01}}
]
properties {
 bOverrideConcurrency=true
 maxDistance=3584
 soundClassObject=&Fort_Player_Foley_1P_Athena
 volumeMultiplier=0.5
 concurrencyOverrides {maxCount=4}
}
