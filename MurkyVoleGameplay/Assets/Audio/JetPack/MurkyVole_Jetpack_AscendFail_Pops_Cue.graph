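# Sound cue for the MurkyVole jetpack's ascend-fail pops, classed as a gadget/item sound.
# A mixer layers a random sputter one-shot with a quieter random exhaust pop, pitched slightly down by a modulator.
# An IsLocal branch routes that shared mix through first-person footstep attenuation for the local player, or third-person footstep attenuation when false or unset.
# The cue itself also references jetpack thruster-start attenuation and overrides concurrency.
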SoundCue /MurkyVoleGameplay/Assets/Audio/JetPack/MurkyVole_Jetpack_AscendFail_Pops_Cue
root=Branch_0
nodes [
 {id=Attenuation_0, children [Mixer_1], properties {attenuationSettings=/Game/Sounds/Fort_Player/Footsteps/Settings/Fort_Player_FS_1P_Att}}
 {id=Attenuation_1, children [Mixer_1], properties {attenuationSettings=/Game/Sounds/Fort_Player/Footsteps/Settings/Fort_Player_FS_3P_Att}}
 {id=Branch_0, children [Attenuation_0, Attenuation_1, Attenuation_1], properties {boolParameterName=IsLocal}}
 {id=Mixer_1, children [Random_2,Modulator_0], properties {inputVolume [0.6,0.4]}}
 {id=Modulator_0, children [Random_0], properties {pitchMax=0.9, pitchMin=0.85}}
 {id=Random_0, children [WavePlayer_0, WavePlayer_9, WavePlayer_11], properties {weights [1,1,1]}}
 {id=Random_2, children [WavePlayer_15, WavePlayer_16, WavePlayer_17], properties {weights [1,1,1]}}
 {id=WavePlayer_0, properties {soundWaveAssetPtr=&/MurkyVole_HoverJets_ExaustPop_01}}
 {id=WavePlayer_11, properties {soundWaveAssetPtr=&/MurkyVole_HoverJets_ExaustPop_04}}
 {id=WavePlayer_15, properties {soundWaveAssetPtr=&/MurkyVole_HoverJets_Sputter_oneShot_01}}
 {id=WavePlayer_16, properties {soundWaveAssetPtr=&/MurkyVole_HoverJets_Sputter_oneShot_02}}
 {id=WavePlayer_17, properties {soundWaveAssetPtr=&/MurkyVole_HoverJets_Sputter_oneShot_03}}
 {id=WavePlayer_9, properties {soundWaveAssetPtr=&/MurkyVole_HoverJets_ExaustPop_02}}
]
properties {
 attenuationSettings=/Game/Sounds/Attenuation/Fort_Jetpack_MurkyVole_Thruster_Start_Att
 bOverrideConcurrency=true
 maxDistance=13000
 soundClassObject=&Fort_Gadgets_Items
 volumeMultiplier=0.5
}
