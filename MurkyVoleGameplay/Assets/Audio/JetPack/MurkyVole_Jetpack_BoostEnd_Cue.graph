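# Sound cue for the MurkyVole jetpack boost-end sound, using an external thruster attenuation asset and the gadgets/items sound class.
# It picks one of three hover-jet boost-off sweetener waves at equal weight and applies a slight random upward pitch variation.
# An IsEnemy branch plays that shared chain directly when true, and through an extra fixed volume cut when false or unset, so enemy jetpacks sound louder.
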SoundCue /MurkyVoleGameplay/Assets/Audio/JetPack/MurkyVole_Jetpack_BoostEnd_Cue
root=Branch_0
nodes [
 {id=Branch_0, children [Modulator_1,Modulator_0,Modulator_0], properties {boolParameterName=IsEnemy}}
 {id=Modulator_0, children [Modulator_1], properties {volumeMax=0.7, volumeMin=0.7}}
 {id=Modulator_1, children [Random_3], properties {pitchMax=1.1, pitchMin=1}}
 {id=Random_3, children [WavePlayer_3, WavePlayer_9, WavePlayer_10], properties {weights [1,1,1]}}
 {id=WavePlayer_10, properties {soundWaveAssetPtr=&/MurkyVole_HoverJets_Boost_Off_Sweetner_03}}
 {id=WavePlayer_3, properties {soundWaveAssetPtr=&/MurkyVole_HoverJets_Boost_Off_Sweetner_01}}
 {id=WavePlayer_9, properties {soundWaveAssetPtr=&/MurkyVole_HoverJets_Boost_Off_Sweetner_02}}
]
properties {
 attenuationSettings=/Game/Sounds/Attenuation/Fort_Jetpack_MurkyVole_Thruster_Start_Att
 maxDistance=13000
 soundClassObject=&Fort_Gadgets_Items
 volumeMultiplier=0.3
}
