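# Sound cue for the MurkyVole jetpack's boost start, playing a random hover-jet one-shot with attenuation as a gadget/item sound.
# An IsEnemy branch plays enemy boosts at full level, while non-enemy or unset cases pass through an extra modulator that lowers volume.
# All paths share a slight upward pitch randomization.
# A quality-level node then picks from six variations at the first two levels, or a three-sample subset at the other two.
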SoundCue /MurkyVoleGameplay/Assets/Audio/JetPack/MurkyVole_Jetpack_BoostStart_Cue
root=Branch_0
nodes [
 {id=Branch_0, children [Modulator_0,Modulator_1,Modulator_1], properties {boolParameterName=IsEnemy}}
 {id=Modulator_0, children [QualityLevel_0], properties {pitchMax=1.1, pitchMin=1}}
 {id=Modulator_1, children [Modulator_0], properties {volumeMax=0.7, volumeMin=0.7}}
 {id=QualityLevel_0, children [Random_0,Random_0,Random_1,Random_1]}
 {id=Random_0, children [WavePlayer_0,WavePlayer_4,WavePlayer_5,WavePlayer_6,WavePlayer_7,WavePlayer_8], properties {weights [1,1,1,1,1,1]}}
 {id=Random_1, children [WavePlayer_6,WavePlayer_7,WavePlayer_8], properties {bShouldExcludeFromBranchCulling=true, weights [1,1,1]}}
 {id=WavePlayer_0, properties {soundWaveAssetPtr=&/MurkyVole_HoverJets_Boost_OneShot_01}}
 {id=WavePlayer_4, properties {soundWaveAssetPtr=&/MurkyVole_HoverJets_Boost_OneShot_02}}
 {id=WavePlayer_5, properties {soundWaveAssetPtr=&/MurkyVole_HoverJets_Boost_OneShot_03}}
 {id=WavePlayer_6, properties {soundWaveAssetPtr=&/MurkyVole_HoverJets_Boost_OneShot_04}}
 {id=WavePlayer_7, properties {soundWaveAssetPtr=&/MurkyVole_HoverJets_Boost_OneShot_05}}
 {id=WavePlayer_8, properties {soundWaveAssetPtr=&/MurkyVole_HoverJets_Boost_OneShot_06}}
]
properties {
 attenuationSettings=/Game/Sounds/Attenuation/Fort_Jetpack_MurkyVole_Thruster_Start_Att
 maxDistance=13000
 soundClassObject=&Fort_Gadgets_Items
 volumeMultiplier=1.5
}
