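# Sound cue for the looping boost sound of the MurkyVole jetpack, classed as a gadget item sound with hover attenuation.
# A looping hover-jet wave player runs through a strengthened Doppler node that feeds two modulators.
# An IsEnemy branch picks the default-volume modulator for enemies and a fixed, quieter one when false or unset, making enemy jetpacks stand out.
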SoundCue /MurkyVoleGameplay/Assets/Audio/JetPack/MurkyVole_Jetpack_Boost_Loop_Cue
root=Branch_0
nodes [
 {id=Branch_0, children [Modulator_0,Modulator_1,Modulator_1], properties {boolParameterName=IsEnemy}}
 {id=Doppler_0, children [WavePlayer_2], properties {dopplerIntensity=2}}
 {id=Modulator_0, children [Doppler_0]}
 {id=Modulator_1, children [Doppler_0], properties {volumeMax=0.7, volumeMin=0.7}}
 {id=WavePlayer_2, properties {bLooping=true, soundWaveAssetPtr=&/MurkyVole_HoverJets_Boost_Loop}}
]
properties {
 attenuationSettings=/Game/Sounds/Attenuation/Fort_Jetpack_MurkyVole_Hover_Att
 maxDistance=10000
 soundClassObject=&Fort_Gadgets_Items
 volumeMultiplier=1.5
}
