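# Sound cue for the jetpack drop impact, built from four hover-jet drop-land wave variations.
# An equally weighted Random node picks one wave and feeds a default Modulator at the root for pitch and volume variation.
# Attenuation and concurrency overrides set a short falloff and cap simultaneous plays, rejecting new ones at the limit.
# It plays at reduced volume under the Fort_Loot_Collect sound class.
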SoundCue /MurkyVoleGameplay/Assets/Audio/JetPack/MurkyVole_Jetpack_DropImpact_Cue
root=Modulator_0
nodes [
 {id=Modulator_0, children [Random_1]}
 {id=Random_1, children [WavePlayer_0,WavePlayer_1,WavePlayer_2,WavePlayer_3], properties {weights [1,1,1,1]}}
 {id=WavePlayer_0, properties {soundWaveAssetPtr=&/MurkyVole_HoverJets_DropLand_01}}
 {id=WavePlayer_1, properties {soundWaveAssetPtr=&/MurkyVole_HoverJets_DropLand_02}}
 {id=WavePlayer_2, properties {soundWaveAssetPtr=&/MurkyVole_HoverJets_DropLand_03}}
 {id=WavePlayer_3, properties {soundWaveAssetPtr=&/MurkyVole_HoverJets_DropLand_04}}
]
properties {
 bOverrideAttenuation=true
 bOverrideConcurrency=true
 maxDistance=1000
 soundClassObject=&Fort_Loot_Collect
 volumeMultiplier=0.7
 attenuationOverrides {falloffDistance=1000, attenuationShapeExtents {x=0}}
 concurrencyOverrides {maxCount=3, resolutionRule=PreventNew}
}
