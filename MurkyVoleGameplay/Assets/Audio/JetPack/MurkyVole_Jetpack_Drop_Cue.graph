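# Sound cue for the Murky Vole jetpack drop sound, built from hover-jet drop recordings.
# A Random node picks among four equally weighted wave variations and feeds a Modulator at the root for pitch and volume variation.
# The cue plays slightly raised in pitch and reduced in volume, using shared weapon-drop attenuation and stop-oldest concurrency assets, under the loot-collect sound class.
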SoundCue /MurkyVoleGameplay/Assets/Audio/JetPack/MurkyVole_Jetpack_Drop_Cue
root=Modulator_1
nodes [
 {id=Modulator_1, children [Random_0]}
 {id=Random_0, children [WavePlayer_0,WavePlayer_1,WavePlayer_2,WavePlayer_3], properties {weights [1,1,1,1]}}
 {id=WavePlayer_0, properties {soundWaveAssetPtr=&/MurkyVole_HoverJets_Drop_01}}
 {id=WavePlayer_1, properties {soundWaveAssetPtr=&/MurkyVole_HoverJets_Drop_02}}
 {id=WavePlayer_2, properties {soundWaveAssetPtr=&/MurkyVole_HoverJets_Drop_03}}
 {id=WavePlayer_3, properties {soundWaveAssetPtr=&/MurkyVole_HoverJets_Drop_04}}
]
properties {
 attenuationSettings=&Fort_Weapon_Drop_Att
 maxDistance=3000
 pitchMultiplier=1.1
 soundClassObject=&Fort_Loot_Collect
 volumeMultiplier=0.6
 concurrencyOverrides {resolutionRule=PreventNew}
 concurrencySet [&Weapon_Drop_StopOldest]
}
