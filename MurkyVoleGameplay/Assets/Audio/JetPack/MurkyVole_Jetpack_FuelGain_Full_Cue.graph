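# Sound cue that plays when the MurkyVole jetpack's fuel finishes refilling, using the hover jets refuel-complete wave.
# A single wave player feeds a modulator that adds a very slight random pitch rise at constant volume.
# The cue is pitched up a little and lowered in volume, routed to the gadgets and items sound class, with jetpack fuel attenuation for distance falloff.
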SoundCue /MurkyVoleGameplay/Assets/Audio/JetPack/MurkyVole_Jetpack_FuelGain_Full_Cue
root=Modulator_2
nodes [
 {id=Modulator_2, children [WavePlayer_1], properties {pitchMax=1.01, pitchMin=1, volumeMax=1, volumeMin=1}}
 {id=WavePlayer_1, properties {soundWaveAssetPtr=&/MurkyVole_HoverJets_Refuel_Complete_03}}
]
properties {
 attenuationSettings=/Game/Sounds/Attenuation/Fort_Jetpack_Fuel_Att
 maxDistance=800
 pitchMultiplier=1.1
 soundClassObject=&Fort_Gadgets_Items
 volumeMultiplier=0.45
}
