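# Sound cue for the looping fuel-gain sound of the MurkyVole jetpack.
# A single looping hover-jet fuel-gain wave feeds a Modulator root whose equal pitch bounds apply a fixed downward pitch shift instead of a random range.
# The cue slightly raises pitch and trims volume, uses the jetpack fuel attenuation asset, and plays in the gadget items sound class.
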SoundCue /MurkyVoleGameplay/Assets/Audio/JetPack/MurkyVole_Jetpack_FuelGain_Loop_Cue
root=Modulator_2
nodes [
 {id=Modulator_2, children [WavePlayer_0], properties {pitchMax=0.6, pitchMin=0.6}}
 {id=WavePlayer_0, properties {bLooping=true, soundWaveAssetPtr=&/MurkyVole_HoverJets_GainFuel_06}}
]
properties {
 attenuationSettings=/Game/Sounds/Attenuation/Fort_Jetpack_Fuel_Att
 maxDistance=800
 pitchMultiplier=1.1
 soundClassObject=&Fort_Gadgets_Items
 volumeMultiplier=0.8
}
