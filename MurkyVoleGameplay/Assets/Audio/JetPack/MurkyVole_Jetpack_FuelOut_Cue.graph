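# Sound cue for the MurkyVole jetpack running out of fuel, playing pitched-up hover-jet deactivate sounds with jetpack attenuation.
# An IsEnemy branch plays enemies at full modulator volume, while false or unset adds a quieter modulator before the shared random-pitch one.
# A quality node picks randomly from five waves for its first two levels and three for the last two.
# Concurrency allows one instance and blocks new ones.
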SoundCue /MurkyVoleGameplay/Assets/Audio/JetPack/MurkyVole_Jetpack_FuelOut_Cue
root=Branch_1
nodes [
 {id=Branch_1, children [Modulator_0,Modulator_2,Modulator_2], properties {boolParameterName=IsEnemy}}
 {id=Modulator_0, children [QualityLevel_0], properties {pitchMax=1.1, pitchMin=1}}
 {id=Modulator_2, children [Modulator_0], properties {pitchMax=1, pitchMin=1, volumeMax=0.7, volumeMin=0.7}}
 {id=QualityLevel_0, children [Random_0,Random_0,Random_1,Random_1]}
 {id=Random_0, children [WavePlayer_1,WavePlayer_2,WavePlayer_3,WavePlayer_4,WavePlayer_5], properties {weights [1,1,1,1,1]}}
 {id=Random_1, children [WavePlayer_3,WavePlayer_4,WavePlayer_5], properties {bShouldExcludeFromBranchCulling=true, weights [1,1,1]}}
 {id=WavePlayer_1, properties {soundWaveAssetPtr=&/MurkyVole_HoverJets_Deactivate_01}}
 {id=WavePlayer_2, properties {soundWaveAssetPtr=&/MurkyVole_HoverJets_Deactivate_02}}
 {id=WavePlayer_3, properties {soundWaveAssetPtr=&/MurkyVole_HoverJets_Deactivate_03}}
 {id=WavePlayer_4, properties {soundWaveAssetPtr=&/MurkyVole_HoverJets_Deactivate_04}}
 {id=WavePlayer_5, properties {soundWaveAssetPtr=&/MurkyVole_HoverJets_Deactivate_05}}
]
properties {
 attenuationSettings=/Game/Sounds/Attenuation/Fort_Jetpack_Att
 bOverrideConcurrency=true
 maxDistance=13067
 pitchMultiplier=1.32
 soundClassObject=&Fort_Gadgets_Items
 subtitlePriority=5000
 volumeMultiplier=0.5
 concurrencyOverrides {maxCount=1, resolutionRule=PreventNew}
}
