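# Sound cue for the MurkyVole jetpack's low-fuel warning, playing a looping hover-jet warning wave.
# A looping Wave Player feeds a Modulator root that pins pitch to a constant, so pitch does not vary.
# It uses the shared Fortnite jetpack attenuation and gadget-item sound class, with boosted volume.
# A concurrency override limits the cue to a single instance at a time.
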SoundCue /MurkyVoleGameplay/Assets/Audio/JetPack/MurkyVole_Jetpack_FuelWarning_Cue
root=Modulator_0
nodes [
 {id=Modulator_0, children [WavePlayer_3], properties {pitchMax=1, pitchMin=1}}
 {
  id=WavePlayer_3
  properties {bLooping=true, soundWaveAssetPtr=&/MurkyVole_HoverJets_LowFuel_Warning_Loop}
 }
]
properties {
 attenuationSettings=/Game/Sounds/Attenuation/Fort_Jetpack_Att
 bOverrideConcurrency=true
 maxDistance=13067
 soundClassObject=&Fort_Gadgets_Items
 volumeMultiplier=1.8
 concurrencyOverrides {maxCount=1}
}
