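# MurkyVole jetpack idle-loop SoundCue, layering hover jets, ground wash and wind movement.
# A hover-attenuated mix drives the idle loop by VelocityMod and fades three ground-disturbance loops out as HeightMod rises, alongside a steady hover loop.
# A separately attenuated wind layer scales with VelocityMod and is quieter unless IsLocal.
# Both feed an IsEnemy branch that plays enemies at full volume and others reduced.
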SoundCue /MurkyVoleGameplay/Assets/Audio/JetPack/MurkyVole_Jetpack_IdleLoop_Cue
root=Modulator_1
nodes [
 {id=Attenuation_0, children [Mixer_0], properties {attenuationSettings=/Game/Sounds/Attenuation/Fort_Jetpack_MurkyVole_Hover_Att}}
 {id=Attenuation_1, children [Branch_0], properties {attenuationSettings=/Game/Sounds/Attenuation/Fort_Jetpack_MurkyVole_3P_Att}}
 {id=Branch_0, children [ModulatorContinuous_6, Modulator_0, Modulator_0], properties {boolParameterName=IsLocal}}
 {id=Branch_1, children [Mixer_1,Modulator_2,Modulator_2], properties {boolParameterName=IsEnemy}}
 {
  id=Mixer_0
  children [ModulatorContinuous_9, WavePlayer_3, ModulatorContinuous_2, ModulatorContinuous_3, ModulatorContinuous_4]
  properties {inputVolume [0.6,1,0.2,0.5,0.5]}
 }
 {id=Mixer_1, children [Attenuation_0, Attenuation_1], properties {inputVolume [1.5,2]}}
 {
  id=ModulatorContinuous_2
  children [WavePlayer_1]
  properties {
   pitchModulationParams {maxInput=1.25, minOutput=1, parameterName=PitchMod}
   volumeModulationParams {default=0, maxInput=0.5, maxOutput=0, minOutput=1, parameterName=HeightMod}
  }
 }
 {
  id=ModulatorContinuous_3
  children [WavePlayer_2]
  properties {
   pitchModulationParams {maxInput=1.25, minOutput=1, parameterName=PitchMod}
   volumeModulationParams {default=0, maxInput=0.5, maxOutput=0, minOutput=1, parameterName=HeightMod}
  }
 }
 {
  id=ModulatorContinuous_4
  children [WavePlayer_4]
  properties {
   pitchModulationParams {maxInput=1.25, minOutput=1, parameterName=PitchMod}
   volumeModulationParams {default=0, maxInput=0.5, maxOutput=0, minOutput=1, parameterName=HeightMod}
  }
 }
 {
  id=ModulatorContinuous_6
  children [WavePlayer_5]
  properties {
   pitchModulationParams {maxInput=1.25, minOutput=1, parameterName=PitchMod}
   volumeModulationParams {default=0, maxInput=700, parameterName=VelocityMod}
  }
 }
 {
  id=ModulatorContinuous_9
  children [WavePlayer_0]
  properties {
   pitchModulationParams {default=0, maxInput=900, maxOutput=1.2, minOutput=1, parameterName=VelocityMod}
   volumeModulationParams {default=0, maxInput=700, minInput=1, minOutput=0.5, parameterName=VelocityMod}
  }
 }
 {id=Modulator_0, children [ModulatorContinuous_6], properties {pitchMax=1, pitchMin=1, volumeMax=0.4, volumeMin=0.4}}
 {id=Modulator_1, children [Branch_1]}
 {id=Modulator_2, children [Mixer_1], properties {pitchMax=1, pitchMin=1, volumeMax=0.7, volumeMin=0.7}}
 {id=WavePlayer_0, properties {bLooping=true, soundWaveAssetPtr=/FNGameplayCues/Athena/Sounds/Items/Jetpack/Jetpack_IdleLoop_01}}
 {
  id=WavePlayer_1
  properties {bLooping=true, soundWaveAssetPtr=&/MurkyBowl_HoverJets_Ground_Disturbance_Loop}
 }
 {
  id=WavePlayer_2
  properties {bLooping=true, soundWaveAssetPtr=&/MurkyBowl_HoverJets_Ground_Disturbance_Debris_Loop}
 }
 {id=WavePlayer_3, properties {bLooping=true, soundWaveAssetPtr=&/MurkyVole_HoverJets_HoverLoop_06}}
 {
  id=WavePlayer_4
  properties {bLooping=true, soundWaveAssetPtr=&/MurkyBowl_HoverJets_Ground_Disturbance_JetWash_Loop}
 }
 {id=WavePlayer_5, properties {bLooping=true, soundWaveAssetPtr=&/MurkyVole_HoverJets_WindMovement_Loop}}
]
properties {maxDistance=10000, soundClassObject=&Fort_Gadgets_Items, volumeMultiplier=1}
