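# Sound cue for the MurkyVole jetpack running out of fuel, with sputters and exhaust pops.
# After a short delay, it mixes a looping sputter, a quieter random sputter one-shot and a random exhaust pop.
# The pop is delayed further and pitched down slightly at random.
# Jetpack attenuation applies, and a concurrency override allows one instance, blocking new plays while it is active.
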SoundCue /MurkyVoleGameplay/Assets/Audio/JetPack/MurkyVole_Jetpack_OutOfFuel_WithPops_Cue
root=Delay_2
nodes [
 {id=Delay_1, children [Modulator_0], properties {delayMax=0.3, delayMin=0.3}}
 {id=Delay_2, children [Mixer_1], properties {delayMax=0.2, delayMin=0.2}}
 {id=Mixer_1, children [Modulator_1,Random_2,Delay_1], properties {inputVolume [1,0.6,1]}}
 {id=Modulator_0, children [Random_0], properties {pitchMax=0.9, pitchMin=0.85}}
 {id=Modulator_1, children [WavePlayer_1], properties {pitchMax=1, pitchMin=1, volumeMax=1, volumeMin=1}}
 {id=Random_0, children [WavePlayer_0, WavePlayer_9, WavePlayer_11], properties {weights [1,1,1]}}
 {id=Random_2, children [WavePlayer_15, WavePlayer_16, WavePlayer_17], properties {weights [1,1,1]}}
 {id=WavePlayer_0, properties {soundWaveAssetPtr=&/MurkyVole_HoverJets_ExaustPop_01}}
 {
  id=WavePlayer_1
  properties {bLooping=true, soundWaveAssetPtr=&/MurkyVole_HoverJets_OutOfFuel_Sputter_Loop}
 }
 {id=WavePlayer_11, properties {soundWaveAssetPtr=&/MurkyVole_HoverJets_ExaustPop_04}}
 {id=WavePlayer_15, properties {soundWaveAssetPtr=&/MurkyVole_HoverJets_Sputter_oneShot_01}}
 {id=WavePlayer_16, properties {soundWaveAssetPtr=&/MurkyVole_HoverJets_Sputter_oneShot_02}}
 {id=WavePlayer_17, properties {soundWaveAssetPtr=&/MurkyVole_HoverJets_Sputter_oneShot_03}}
 {id=WavePlayer_9, properties {soundWaveAssetPtr=&/MurkyVole_HoverJets_ExaustPop_02}}
]
properties {
 attenuationSettings=/Game/Sounds/Attenuation/Fort_Jetpack_Att
 bOverrideConcurrency=true
 maxDistance=13067
 soundClassObject=&Fort_Gadgets_Items
 volumeMultiplier=1.5
 concurrencyOverrides {maxCount=1, resolutionRule=PreventNew}
}
