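# Sound cue for picking up the jetpack (HoverJets), classed as loot collection with weapon-pickup attenuation.
# A mixer layers an immediate equip sound with two equally delayed layers: a half-level random pick of three pickup variations and an equip servo.
# Modulators leave default variation on the equip sound, fix pitch on the pickup layer, and fix both pitch and volume on the servo.
# The cue references a shared pickup concurrency set.
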SoundCue /MurkyVoleGameplay/Assets/Audio/JetPack/MurkyVole_Jetpack_Pickup_Cue
root=Mixer_1
nodes [
 {id=Delay_0, children [Modulator_1], properties {delayMax=0.8, delayMin=0.8}}
 {id=Delay_1, children [Modulator_2], properties {delayMax=0.8, delayMin=0.8}}
 {id=Mixer_1, children [Modulator_0,Delay_0,Delay_1], properties {inputVolume [1,0.5,1]}}
 {id=Modulator_0, children [WavePlayer_25]}
 {id=Modulator_1, children [Random_2], properties {pitchMax=1, pitchMin=1}}
 {id=Modulator_2, children [WavePlayer_26], properties {pitchMax=1, pitchMin=1, volumeMax=1, volumeMin=1}}
 {id=Random_2, children [WavePlayer_21, WavePlayer_22, WavePlayer_23], properties {weights [1,1,1]}}
 {id=WavePlayer_21, properties {soundWaveAssetPtr=&/MurkyVole_HoverJets_PickUp_04}}
 {id=WavePlayer_22, properties {soundWaveAssetPtr=&/MurkyVole_HoverJets_PickUp_03}}
 {id=WavePlayer_23, properties {soundWaveAssetPtr=&/MurkyVole_HoverJets_PickUp_02}}
 {id=WavePlayer_25, properties {soundWaveAssetPtr=&/MurkyVole_HoverJets_Equip_01}}
 {id=WavePlayer_26, properties {soundWaveAssetPtr=&/MurkyVole_HoverJets_Equip_Servo}}
]
properties {
 attenuationSettings=&Fort_Weapon_Pickup_Att
 maxDistance=3500
 soundClassObject=&Fort_Loot_Collect
 volumeMultiplier=0.5
 concurrencyOverrides {maxCount=1, resolutionRule=PreventNew}
 concurrencySet [&PickupMaximum_Oldest]
}
