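# Sound cue for the MurkyVole jetpack start, layering a stock Jetpack_Start wave with a random HoverJets activation variant.
# An IsEnemy branch plays the pitch-randomized mix directly when true, and through an extra volume-reducing modulator when false or unset.
# A quality-level node picks from five variants on the first two levels and three on the last two.
# Thruster-start attenuation applies, and a single-instance concurrency override rejects overlapping plays.
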SoundCue /MurkyVoleGameplay/Assets/Audio/JetPack/MurkyVole_Jetpack_Start_Cue
root=Branch_1
nodes [
 {id=Branch_1, children [Modulator_0,Modulator_2,Modulator_2], properties {boolParameterName=IsEnemy}}
 {id=Mixer_0, children [WavePlayer_1, QualityLevel_0], properties {inputVolume [1,1]}}
 {id=Modulator_0, children [Mixer_0], properties {pitchMax=1.1, pitchMin=1}}
 {id=Modulator_2, children [Modulator_0], properties {pitchMax=1, pitchMin=1, volumeMax=0.7, volumeMin=0.7}}
 {id=QualityLevel_0, children [Random_0,Random_0,Random_1,Random_1]}
 {id=Random_0, children [WavePlayer_0,WavePlayer_2,WavePlayer_3,WavePlayer_4,WavePlayer_5], properties {weights [1,1,1,1,1]}}
 {id=Random_1, children [WavePlayer_3,WavePlayer_4,WavePlayer_5], properties {bShouldExcludeFromBranchCulling=true, weights [1,1,1]}}
 {id=WavePlayer_0, properties {soundWaveAssetPtr=&/MurkyVole_HoverJets_Activate_01}}
 {id=WavePlayer_1, properties {soundWaveAssetPtr=/FNGameplayCues/Athena/Sounds/Items/Jetpack/Jetpack_Start_01}}
 {id=WavePlayer_2, properties {soundWaveAssetPtr=&/MurkyVole_HoverJets_Activate_02}}
 {id=WavePlayer_3, properties {soundWaveAssetPtr=&/MurkyVole_HoverJets_Activate_03}}
 {id=WavePlayer_4, properties {soundWaveAssetPtr=&/MurkyVole_HoverJets_Activate_04}}
 {id=WavePlayer_5, properties {soundWaveAssetPtr=&/MurkyVole_HoverJets_Activate_05}}
]
properties {
 attenuationSettings=/Game/Sounds/Attenuation/Fort_Jetpack_MurkyVole_Thruster_Start_Att
 bOverrideConcurrency=true
 maxDistance=13000
 soundClassObject=&Fort_Gadgets_Items
 volumeMultiplier=0.6
 concurrencyOverrides {maxCount=1, resolutionRule=PreventNew}
}
